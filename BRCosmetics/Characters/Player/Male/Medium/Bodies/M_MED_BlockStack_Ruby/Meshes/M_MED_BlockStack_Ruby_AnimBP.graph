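# Post-process style body AnimBlueprint for the BlockStack Ruby cosmetic, layering the head mesh's pose over the parent mesh's pose on selected bone branches.
# A weighted multi-bone constraint and two additive offsets then adjust the right pouch.
# The Head reference comes from customization loading, or on mannequins from a per-update component search for Hope or ChunkBuggy head meshes.
# CrouchingIdle is updated but unread here.
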AnimBlueprint /BRCosmetics/Characters/Player/Male/Medium/Bodies/M_MED_BlockStack_Ruby/Meshes/M_MED_BlockStack_Ruby_AnimBP
target_skeleton=&/M_MED_BlockStack_Ruby_Skeleton
graphs [
 AnimGraph {nodes [
   n1 = Layered blend per bone(
    BasePose=Copy Pose From Mesh() {bUseAttachedParent=true}
    BlendPoses_0=Copy Pose From Mesh(SourceMeshComponent=Get Head() -> (Head:Skeletal Mesh Component Budgeted Object Reference)) {bUseAttachedParent=true}
   ) {layerSetup [{branchFilters [
       {boneName=attach_backpack}
       {boneName=attach_cape}
       {boneName=deform_foot_l}
       {boneName=deform_ball_l}
       {boneName=deform_foot_r}
       {boneName=deform_ball_r}
       {boneName=foot_bk_r}
       {boneName=foot_fr_r}
       {boneName=foot_bk_l}
       {boneName=foot_fr_l}
       {boneName=deform_pelvis}
       {boneName=pouch_r}
       {boneName=deform_thigh_r}
      ]}]}
   n2 = "Constraint - pouch_r (PPPPPPP)"!(
    ComponentPose=Local To Component(LocalPose=@n1)
    Alpha=1
    ConstraintWeights_0=0.6681
    ConstraintWeights_1:float=0.1525
    ConstraintWeights_2:float=0.0761
    ConstraintWeights_3:float=0.0408
    ConstraintWeights_4:float=0.0307
    ConstraintWeights_5:float=0.0282
    ConstraintWeights_6:float=0.0034
   ) -> (Pose) {
    boneToModify {boneName=pouch_r}
    constraintSetup [
     {targetBone {boneName=deform_thighUp_r}}
     {targetBone {boneName=deform_thigh_bk_r}}
     {targetBone {boneName=deform_pelvis}}
     {targetBone {boneName=deform_thigh_fr_r}}
     {targetBone {boneName=thigh_r}}
     {targetBone {boneName=thigh_twist_01_r}}
     {targetBone {boneName=deform_knee_in_r}}
    ]
    constraintWeights [1,1,1,1,1,1,1]
   }
   n3 = "Transform (Modify) Bone - Bone: pouch_r"!(ComponentPose=@n2, bAlphaBoolEnabled:bool=True, Translation="-0.000,0.118,-0.516", Rotation="-0.000,0.011,-12.801", Scale="1,1,1") -> (Pose) {
    alphaInputType=Bool
    rotationMode=BMM_Additive
    rotationSpace=BCS_BoneSpace
    scaleMode=BMM_Additive
    scaleSpace=BCS_BoneSpace
    translationMode=BMM_Additive
    translationSpace=BCS_BoneSpace
    alphaBoolBlend {blendInTime=0.2, blendOutTime=0.2}
    boneToModify {boneName=pouch_r}
    rotation {roll=-12.801, yaw=0.011}
    translation {y=0.118, z=-0.516}
   }
   n4 = "Transform (Modify) Bone - Bone: pouch_r"(ComponentPose=@n3, bAlphaBoolEnabled=True, Translation="-0.126,-0.051,-0.376", Rotation="-3.944,0.002,-1.622", Scale="1,1,1") {
    alphaInputType=Bool
    rotationMode=BMM_Additive
    rotationSpace=BCS_BoneSpace
    scaleMode=BMM_Additive
    scaleSpace=BCS_BoneSpace
    translationMode=BMM_Additive
    translationSpace=BCS_BoneSpace
    alphaBoolBlend {blendInTime=0.2, blendOutTime=0.2}
    boneToModify {boneName=pouch_r}
    rotation=<rot -3.944 0.002 -1.622>
    translation=<-0.126 -0.051 -0.376>
   }
   Output Pose(Result=Component To Local(ComponentPose=@n4)) {layerGroup=None}
  ]}
 EventGraph {nodes [
   Event BlueprintUpdateAnimation()
   > n1 = Sequence()
   n2 = GetOwningActor()
   > Cast To FortPlayerMannequin!(execute, Object=@n2) -> (then, CastFailed, AsFort Player Mannequin:Fort Player Mannequin Object Reference) {is=title K2Node_DynamicCast}
   > n3 = For Each Loop(Array:Array of Skeletal Mesh Component Object References=Get Components By Class!(self:Actor Object Reference=@n2, ComponentClass:Actor Component Class Reference=/Script/Engine.SkeletalMeshComponent) -> (ReturnValue:Array of Skeletal Mesh Component Object References)) -> (Array Element:Skeletal Mesh Component Object Reference)
   n4 = Get SkeletalMesh!(self:Skinned Mesh Component Object Reference=@n3) -> (SkeletalMesh:Skeletal Mesh Object Reference)
   > Branch(Condition=OR Boolean(
     A="Equal (Object)"(A=@n4, B=/BRCosmetics/Characters/Player/Male/Medium/Bodies/M_MED_BlockStack_Hope/Meshes/M_MED_BlockStack_Hope_Head)
     B="Equal (Object)"(A=@n4, B=/BRCosmetics/Characters/Player/Male/Medium/Bodies/M_MED_BlockStack_ChunkBuggy/Meshes/M_MED_Blockstack_ChunkBuggy_Head)
    ))
   > n5 = Cast To SkeletalMeshComponentBudgeted!(execute, Object=@n3) -> (then, CastFailed, AsSkeletal Mesh Component Budgeted:Skeletal Mesh Component Budgeted Object Reference) {is=title K2Node_DynamicCast}
   > Set Head(Head:Skeletal Mesh Component Budgeted Object Reference=@n5) -> (Output_Get:Skeletal Mesh Component Budgeted Object Reference)
   n6 = Get bIsMoving2D()
   Set CrouchingIdle(execute=@n1.then_1, CrouchingIdle:bool=AND Boolean(A=Get bIsCrouching(), B=NOT Boolean(A=@n6))) -> (Output_Get:bool)
   n7 = Event OnCharacterCustomizationLoaded()
   > Set Head(Head=Get Array Item with EFort Custom Part Type(Array=@n7.CustomMeshParts, PartType=Head))
  ]}
]
variables [
 CrouchingIdle {type=boolean}
 Head {type=SkeletalMeshComponentBudgeted}
]
