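# Physics asset for the face accessory mesh of the BlockStack Ruby medium male player body, covering a left-side dynamic hair chain.
# Three dyn_hair capsules chain from base_hair_l with narrow swing limits and locked twist.
# Boxes on head and upper spine, plus capsules on clavicles and left arm, form body shapes linked back to spine_04 by angularly free constraints.
# The hair chain has no constraint to those body shapes.
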PhysicsAsset /BRCosmetics/Characters/Player/Male/Medium/Bodies/M_MED_BlockStack_Ruby/Meshes/Parts/M_MED_Blockstack_Ruby_FaceAcc_Physics
bodies {
 arm_bend_1_l [
  {shapeType=Capsule, length=10.5726, radius=2, shapeName=arm_bend_1_l_capsule, center=<6.3942 0 0>, rotation=<rot 90 0 0>}
  {shapeType=Capsule, length=15, radius=3.5, shapeName=arm_bend_1_l_capsule, center=<7.4525 2.712 -0.8095>, rotation=<rot 90 0 0>}
  {shapeType=Capsule, length=15.4638, radius=3.0704, shapeName=arm_bend_1_l_capsule, center=<7.4525 6.7035 -0.4957>, rotation=<rot 90 0 0>}
 ]
 arm_bend_3_l [{shapeType=Capsule, length=16.7718, radius=4, shapeName=arm_bend_3_l_capsule, center=<0.2425 0 0>, rotation=<rot 90 0 0>}]
 arm_bend_4_l [
  {shapeType=Capsule, length=7.0312, radius=4, shapeName=arm_bend_4_l_capsule, center=<2.6731 0 0>, rotation=<rot 90 0 0>}
  {shapeType=Capsule, length=15, radius=4.3404, shapeName=arm_bend_4_l_capsule, center=<0 5.3252 -6.1929>, rotation=<rot 90 0 0>}
 ]
 base_hair_l [{shapeType=Capsule, length=0.505, radius=0.505, shapeName=base_hair_l_capsule, rotation=<rot 0 -90 0>}]
 clavicle_l [{shapeType=Capsule, length=0.5, radius=0.5, shapeName=clavicle_l_capsule}]
 clavicle_r [{shapeType=Capsule, length=0.505, radius=0.505, shapeName=clavicle_r_capsule, rotation=<rot 0 -90 0>}]
 deform_head [{shapeType=Box, extentX=50.4675, extentY=33.2222, extentZ=26.1874, shapeName=deform_head_box, center=<-1.7767 0 0>}]
 dyn_hair_01_l [{
   shapeType=Capsule
   length=5.9952
   radius=3
   shapeName=dyn_hair_01_l_capsule
   center=<5.4004 -1.9165 0.4255>
   rotation=<rot 80.5182 90.0001 0>
  }]
 dyn_hair_02_l [{
   shapeType=Capsule
   length=5.026
   radius=3.01
   shapeName=dyn_hair_02_l_capsule
   center=<5.6952 0.0478 -0.2786>
   rotation=<rot 81.7762 90.0001 0>
  }]
 dyn_hair_03_l [{shapeType=Capsule, length=6.5089, radius=1.233, shapeName=dyn_hair_03_l_capsule, center=<3.9804 2.095 0>, rotation=<rot 90 0 0>}]
 spine_04 [{
   shapeType=Box
   extentX=30.2044
   extentY=31.4031
   extentZ=28.3877
   shapeName=spine_04_box
   center=<-0.0832 1.1159 0>
   rotation=<rot 0 4.2659 0>
  }]
 upperRetArm_l_base [{shapeType=Capsule, length=5, radius=5, shapeName=upperRetArm_l_base_capsule, rotation=<rot 90 0 0>}]
}
constraints [
 {bone1Name=dyn_hair_03_l, bone2Name=dyn_hair_02_l, swing1LimitDegrees=10, swing2LimitDegrees=10, twistLimitDegrees=45, twistMotion=Locked}
 {bone1Name=dyn_hair_02_l, bone2Name=dyn_hair_01_l, swing1LimitDegrees=10, swing2LimitDegrees=10, twistLimitDegrees=45, twistMotion=Locked}
 {bone1Name=dyn_hair_01_l, bone2Name=base_hair_l, swing1LimitDegrees=15, swing2LimitDegrees=10, twistLimitDegrees=45, twistMotion=Locked}
 {
  bone1Name=arm_bend_1_l
  bone2Name=upperRetArm_l_base
  swing1LimitDegrees=45
  swing1Motion=Free
  swing2LimitDegrees=45
  swing2Motion=Free
  twistLimitDegrees=45
  twistMotion=Free
 }
 {
  bone1Name=arm_bend_3_l
  bone2Name=upperRetArm_l_base
  swing1LimitDegrees=45
  swing1Motion=Free
  swing2LimitDegrees=45
  swing2Motion=Free
  twistLimitDegrees=45
  twistMotion=Free
 }
 {
  bone1Name=arm_bend_4_l
  bone2Name=upperRetArm_l_base
  swing1LimitDegrees=45
  swing1Motion=Free
  swing2LimitDegrees=45
  swing2Motion=Free
  twistLimitDegrees=45
  twistMotion=Free
 }
 {
  bone1Name=upperRetArm_l_base
  bone2Name=spine_04
  swing1LimitDegrees=45
  swing1Motion=Free
  swing2LimitDegrees=45
  swing2Motion=Free
  twistLimitDegrees=45
  twistMotion=Free
 }
 {
  bone1Name=clavicle_l
  bone2Name=spine_04
  swing1LimitDegrees=45
  swing1Motion=Free
  swing2LimitDegrees=45
  swing2Motion=Free
  twistLimitDegrees=45
  twistMotion=Free
 }
 {
  bone1Name=clavicle_r
  bone2Name=spine_04
  swing1LimitDegrees=45
  swing1Motion=Free
  swing2LimitDegrees=45
  swing2Motion=Free
  twistLimitDegrees=45
  twistMotion=Free
 }
 {
  bone1Name=deform_head
  bone2Name=spine_04
  swing1LimitDegrees=45
  swing1Motion=Free
  swing2LimitDegrees=45
  swing2Motion=Free
  twistLimitDegrees=45
  twistMotion=Free
 }
]
