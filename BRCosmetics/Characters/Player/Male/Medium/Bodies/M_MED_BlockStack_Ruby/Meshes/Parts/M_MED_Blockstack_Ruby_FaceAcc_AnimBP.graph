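# Post-process-style part AnimBlueprint for the Blockstack Ruby character's face accessory, adding secondary motion to its hair.
# The attached parent's copied pose feeds an emote-aware Control Rig, a pelvis look-at and additive rotation on base_hair_l, then RigidBody simulation relative to deform_head.
# Each update computes gravity and rigid-body settings from a parameter asset and sets an upper-arm alpha from parachute and emote state.
# Parachute events toggle a hang-gliding alpha.
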AnimBlueprint /BRCosmetics/Characters/Player/Male/Medium/Bodies/M_MED_BlockStack_Ruby/Meshes/Parts/M_MED_Blockstack_Ruby_FaceAcc_AnimBP
target_skeleton=&/M_MED_Blockstack_Ruby_FaceAcc_Skeleton
graphs [
 AnimGraph {nodes [
   n1 = Control Rig!(Source=Copy Pose From Mesh() {bUseAttachedParent=true}, Alpha, Emote:bool=Get bIsPlayingEmote()) -> (Pose) {controlRigAssetReference {controlRigAsset=&/M_MED_Blockstack_Ruby_FaceAcc_CtrlRig}}
   n2 = Look At - Bone: base_hair_l(ComponentPose=Local To Component(LocalPose=@n1), Alpha=1) {
    boneToModify {boneName=base_hair_l}
    lookAtLocation=<0 13.273 -28.2975>
    lookAtTarget {boneReference {boneName=pelvis}}
    lookAt_Axis {axis=<1 0>}
   }
   n3 = "Transform (Modify) Bone - Bone: base_hair_l"!(ComponentPose=@n2, bAlphaBoolEnabled:bool=True, Translation="0,0,0", Rotation="-10,30,0", Scale="1,1,1") -> (Pose) {
    alphaInputType=Bool
    rotationMode=BMM_Additive
    rotationSpace=BCS_BoneSpace
    translationMode=BMM_Additive
    translationSpace=BCS_BoneSpace
    alphaBoolBlend {blendInTime=0.2, blendOutTime=0.2}
    boneToModify {boneName=base_hair_l}
    rotation {pitch=-10, yaw=30}
   }
   n4 = RigidBody!(
    ComponentPose=@n3
    Alpha
    OverrideWorldGravity:Vector="0,0,-980"
    ExternalForce
    SimSpaceSettings:Sim Space Settings Structure=(WorldAlpha=1,VelocityScaleZ=1,DampingAlpha=1,MaxLinearVelocity=100,MaxAngularVelocity=15,MaxLinearAcceleration=200,MaxAngularAcceleration=50,ExternalLinearDragV=(X=0,Y=0,Z=0),ExternalLinearVelocity=(X=0,Y=0,Z=0),ExternalAngularVelocity=(X=0,Y=0,Z=0))
   ) -> (Pose) {
    bOverrideWorldGravity=true
    overridePhysicsAsset=&/M_MED_Blockstack_Ruby_FaceAcc_Physics
    simulationSpace=BaseBoneSpace
    baseBoneRef {boneName=deform_head}
    componentAppliedLinearAccClamp=<0 0 0>
    overrideWorldGravity {z=-980}
    simSpaceSettings {maxAngularAcceleration=50, maxAngularVelocity=15, maxLinearAcceleration=200, maxLinearVelocity=100, worldAlpha=1}
   }
   Output Pose(Result=Component To Local(ComponentPose=@n4)) {layerGroup=None}
  ]}
 EventGraph {nodes [
   Event BlueprintUpdateAnimation()
   > n1 = Sequence()
   > n2 = "Gravity Override State Function (Parameter Asset) (Nativized)"(
    CharacterPartAnimInstance=Self-Reference()
    GameTimeInSeconds=GetGameTimeInSeconds()
    DebugJointName=spine_05
    Parameters=&/M_MED_Blockstack_Ruby_FaceAcc_Parameters
   )
   > Set Gravity Override(Gravity Override=@n2)
   > Set RBAN Settings Out(RBAN Settings Out=@n2.OutRigidBodyAnimNodeParameters)
   n3 = Branch(execute=@n1.then_1, Condition=Get bIsParachuteOpened())
   > Set Alpha_UpperarmUP(Alpha_UpperarmUP:double=0.5) -> (Output_Get:double)
   n4 = Branch(execute=@n3.else, Condition=Get bIsPlayingEmote())
   > Set Alpha_UpperarmUP(Alpha_UpperarmUP=GetCurveValue(CurveName=UpperarmUp))
   Set Alpha_UpperarmUP(execute=@n4.else, Alpha_UpperarmUP=1)
   Event OnParachuteOpened()
   > Switch on EGliderType(Selection=Get PlayerGliderType() -> (PlayerGliderType=HangGlider))
   > Set HangGlidingAlpha(HangGlidingAlpha:double=1) -> (Output_Get:double)
   Event OnParachuteClosed()
   > Set HangGlidingAlpha(HangGlidingAlpha=0)
  ]}
]
variables [
 Alpha_UpperarmUP {type=number}
 Gravity Override {type=Vector}
 HangGlidingAlpha {type=number}
 RBAN Settings Out {
  type=FortRigidBodyAnimNodeParameters
  default {simSpaceSettings {
    dampingAlpha=0
    maxAngularAcceleration=100
    maxAngularVelocity=5
    maxLinearAcceleration=500
    maxLinearVelocity=350
    worldAlpha=1
    externalLinearDragV=<1 1 1>
   }}
 }
]
